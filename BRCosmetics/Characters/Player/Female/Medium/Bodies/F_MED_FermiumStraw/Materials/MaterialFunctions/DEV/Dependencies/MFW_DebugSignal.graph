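# MFW_DebugSignal is a wrapper material function that turns scalar, Vector2, Vector3 or Vector4 signals into Vector3 debug outputs through four MF_DebugSignal calls.
# It pads smaller inputs to Vector4 and declares named reroutes for shared zz_SIGNAL DEBUG parameters: swizzle, remaps, maths, tint and sRGB-to-linear.
# Remap and sRGB parameters use a rounded, saturated switch to lerp between one RGB value and per-channel values, with OldMax sharing OldMin's switch.
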MaterialFunction /BRCosmetics/Characters/Player/Female/Medium/Bodies/F_MED_FermiumStraw/Materials/MaterialFunctions/DEV/Dependencies/MFW_DebugSignal
graph {expressions {
  MaterialExpressionAdd_2 {
   class=MaterialExpressionAdd
   inputs [
    A {from=MaterialExpressionVectorParameter_7, output=RGB}
    B {from=MaterialExpressionVectorParameter_7, output=RGB}
   ]
  }
  MaterialExpressionConvert_0 {
   class=MaterialExpressionConvert
   inputs [
    None {from=MaterialExpressionVectorParameter_9, output=Invert Sign RGB / Invert Sign R}
    None {from=MaterialExpressionVectorParameter_9, output=Invert Sign RGB / Invert Sign R}
    None {from=MaterialExpressionVectorParameter_9, output=Invert Sign RGB / Invert Sign R}
   ]
   properties {
    nodeName=Make Vector3
    convertInputs [
     {
      type=Scalar
      expressionInput {
       expression=&self:MaterialExpressionVectorParameter_9
       inputName=None
       mask=1
       maskA=0
       maskB=0
       maskG=1
       maskR=0
       outputIndex=2
      }
     }
     {
      type=Scalar
      expressionInput {
       expression=&self:MaterialExpressionVectorParameter_9
       inputName=None
       mask=1
       maskA=0
       maskB=1
       maskG=0
       maskR=0
       outputIndex=3
      }
     }
     {
      type=Scalar
      expressionInput {
       expression=&self:MaterialExpressionVectorParameter_9
       inputName=None
       mask=1
       maskA=1
       maskB=0
       maskG=0
       maskR=0
       outputIndex=4
      }
     }
    ]
    convertOutputs [{type=Vector3}]
   }
  }
  MaterialExpressionConvert_1 {
   class=MaterialExpressionConvert
   inputs [None {from=MaterialExpressionVectorParameter_9, output=Invert Sign RGB / Invert Sign R}]
   properties {
    convertInputs [{
      type=Scalar
      expressionInput {
       expression=&self:MaterialExpressionVectorParameter_9
       inputName=None
       mask=1
       maskA=0
       maskB=0
       maskG=1
       maskR=0
       outputIndex=2
      }
     }]
    convertOutputs [{type=Vector3}]
   }
  }
  MaterialExpressionConvert_11 {
   class=MaterialExpressionConvert
   inputs [
    None {from=MaterialExpressionVectorParameter_8, output=Convert RGB / Convert R}
    None {from=MaterialExpressionVectorParameter_8, output=Convert RGB / Convert R}
    None {from=MaterialExpressionVectorParameter_8, output=Convert RGB / Convert R}
   ]
   properties {
    nodeName=Make Vector3
    convertInputs [
     {
      type=Scalar
      expressionInput {
       expression=&self:MaterialExpressionVectorParameter_8
       inputName=None
       mask=1
       maskA=0
       maskB=0
       maskG=1
       maskR=0
       outputIndex=2
      }
     }
     {
      type=Scalar
      expressionInput {
       expression=&self:MaterialExpressionVectorParameter_8
       inputName=None
       mask=1
       maskA=0
       maskB=1
       maskG=0
       maskR=0
       outputIndex=3
      }
     }
     {
      type=Scalar
      expressionInput {
       expression=&self:MaterialExpressionVectorParameter_8
       inputName=None
       mask=1
       maskA=1
       maskB=0
       maskG=0
       maskR=0
       outputIndex=4
      }
     }
    ]
    convertOutputs [{type=Vector3}]
   }
  }
  MaterialExpressionConvert_12 {
   class=MaterialExpressionConvert
   inputs [None {from=MaterialExpressionVectorParameter_8, output=Convert RGB / Convert R}]
   properties {
    convertInputs [{
      type=Scalar
      expressionInput {
       expression=&self:MaterialExpressionVectorParameter_8
       inputName=None
       mask=1
       maskA=0
       maskB=0
       maskG=1
       maskR=0
       outputIndex=2
      }
     }]
    convertOutputs [{type=Vector3}]
   }
  }
  MaterialExpressionConvert_13 {
   class=MaterialExpressionConvert
   inputs [
    None {from=MaterialExpressionVectorParameter_11, output=One Minus RGB / One Minus R}
    None {from=MaterialExpressionVectorParameter_11, output=One Minus RGB / One Minus R}
    None {from=MaterialExpressionVectorParameter_11, output=One Minus RGB / One Minus R}
   ]
   properties {
    nodeName=Make Vector3
    convertInputs [
     {
      type=Scalar
      expressionInput {
       expression=&self:MaterialExpressionVectorParameter_11
       inputName=None
       mask=1
       maskA=0
       maskB=0
       maskG=1
       maskR=0
       outputIndex=2
      }
     }
     {
      type=Scalar
      expressionInput {
       expression=&self:MaterialExpressionVectorParameter_11
       inputName=None
       mask=1
       maskA=0
       maskB=1
       maskG=0
       maskR=0
       outputIndex=3
      }
     }
     {
      type=Scalar
      expressionInput {
       expression=&self:MaterialExpressionVectorParameter_11
       inputName=None
       mask=1
       maskA=1
       maskB=0
       maskG=0
       maskR=0
       outputIndex=4
      }
     }
    ]
    convertOutputs [{type=Vector3}]
   }
  }
  MaterialExpressionConvert_16 {
   class=MaterialExpressionConvert
   inputs [None {from=MaterialExpressionVectorParameter_11, output=One Minus RGB / One Minus R}]
   properties {
    convertInputs [{
      type=Scalar
      expressionInput {
       expression=&self:MaterialExpressionVectorParameter_11
       inputName=None
       mask=1
       maskA=0
       maskB=0
       maskG=1
       maskR=0
       outputIndex=2
      }
     }]
    convertOutputs [{type=Vector3}]
   }
  }
  MaterialExpressionConvert_17 {
   class=MaterialExpressionConvert
   inputs [
    None {from=MaterialExpressionVectorParameter_2, output=OldMin RGB / OldMin R}
    None {from=MaterialExpressionVectorParameter_2, output=OldMin RGB / OldMin R}
    None {from=MaterialExpressionVectorParameter_2, output=OldMin RGB / OldMin R}
   ]
   properties {
    nodeName=Make Vector3
    convertInputs [
     {
      type=Scalar
      expressionInput {
       expression=&self:MaterialExpressionVectorParameter_2
       inputName=None
       mask=1
       maskA=0
       maskB=0
       maskG=1
       maskR=0
       outputIndex=2
      }
     }
     {
      type=Scalar
      expressionInput {
       expression=&self:MaterialExpressionVectorParameter_2
       inputName=None
       mask=1
       maskA=0
       maskB=1
       maskG=0
       maskR=0
       outputIndex=3
      }
     }
     {
      type=Scalar
      expressionInput {
       expression=&self:MaterialExpressionVectorParameter_2
       inputName=None
       mask=1
       maskA=1
       maskB=0
       maskG=0
       maskR=0
       outputIndex=4
      }
     }
    ]
    convertOutputs [{type=Vector3}]
   }
  }
  MaterialExpressionConvert_18 {
   class=MaterialExpressionConvert
   inputs [None {from=MaterialExpressionVectorParameter_2, output=OldMin RGB / OldMin R}]
   properties {
    convertInputs [{
      type=Scalar
      expressionInput {
       expression=&self:MaterialExpressionVectorParameter_2
       inputName=None
       mask=1
       maskA=0
       maskB=0
       maskG=1
       maskR=0
       outputIndex=2
      }
     }]
    convertOutputs [{type=Vector3}]
   }
  }
  MaterialExpressionConvert_19 {
   class=MaterialExpressionConvert
   inputs [
    None {from=MaterialExpressionVectorParameter_3, output=OldMax RGB / OldMax R}
    None {from=MaterialExpressionVectorParameter_3, output=OldMax RGB / OldMax R}
    None {from=MaterialExpressionVectorParameter_3, output=OldMax RGB / OldMax R}
   ]
   properties {
    nodeName=Make Vector3
    convertInputs [
     {
      type=Scalar
      expressionInput {
       expression=&self:MaterialExpressionVectorParameter_3
       inputName=None
       mask=1
       maskA=0
       maskB=0
       maskG=1
       maskR=0
       outputIndex=2
      }
     }
     {
      type=Scalar
      expressionInput {
       expression=&self:MaterialExpressionVectorParameter_3
       inputName=None
       mask=1
       maskA=0
       maskB=1
       maskG=0
       maskR=0
       outputIndex=3
      }
     }
     {
      type=Scalar
      expressionInput {
       expression=&self:MaterialExpressionVectorParameter_3
       inputName=None
       mask=1
       maskA=1
       maskB=0
       maskG=0
       maskR=0
       outputIndex=4
      }
     }
    ]
    convertOutputs [{type=Vector3}]
   }
  }
  MaterialExpressionConvert_2 {
   class=MaterialExpressionConvert
   inputs [
    None {from=MaterialExpressionFunctionInput_0}
    None {from=""}
   ]
   properties {
    nodeName=Make Vector4
    convertInputs [
     {
      type=Scalar
      expressionInput {
       expression=&self:MaterialExpressionFunctionInput_0
       inputName=None
       mask=0
       maskA=0
       maskB=0
       maskG=0
       maskR=0
       outputIndex=0
      }
     }
     {type=Scalar, expressionInput {expression=null, inputName=None, mask=0, maskA=0, maskB=0, maskG=0, maskR=0, outputIndex=0}}
    ]
    convertOutputs [{type=Vector4, defaultValue=<rgba 0 0 0 1>}]
   }
  }
  MaterialExpressionConvert_20 {
   class=MaterialExpressionConvert
   inputs [None {from=MaterialExpressionVectorParameter_3, output=OldMax RGB / OldMax R}]
   properties {
    convertInputs [{
      type=Scalar
      expressionInput {
       expression=&self:MaterialExpressionVectorParameter_3
       inputName=None
       mask=1
       maskA=0
       maskB=0
       maskG=1
       maskR=0
       outputIndex=2
      }
     }]
    convertOutputs [{type=Vector3}]
   }
  }
  MaterialExpressionConvert_21 {
   class=MaterialExpressionConvert
   inputs [
    None {from=MaterialExpressionVectorParameter_10, output=Domain RGB / Domain R}
    None {from=MaterialExpressionVectorParameter_10, output=Domain RGB / Domain R}
    None {from=MaterialExpressionVectorParameter_10, output=Domain RGB / Domain R}
   ]
   properties {
    nodeName=Make Vector3
    convertInputs [
     {
      type=Scalar
      expressionInput {
       expression=&self:MaterialExpressionVectorParameter_10
       inputName=None
       mask=1
       maskA=0
       maskB=0
       maskG=1
       maskR=0
       outputIndex=2
      }
     }
     {
      type=Scalar
      expressionInput {
       expression=&self:MaterialExpressionVectorParameter_10
       inputName=None
       mask=1
       maskA=0
       maskB=1
       maskG=0
       maskR=0
       outputIndex=3
      }
     }
     {
      type=Scalar
      expressionInput {
       expression=&self:MaterialExpressionVectorParameter_10
       inputName=None
       mask=1
       maskA=1
       maskB=0
       maskG=0
       maskR=0
       outputIndex=4
      }
     }
    ]
    convertOutputs [{type=Vector3}]
   }
  }
  MaterialExpressionConvert_22 {
   class=MaterialExpressionConvert
   inputs [None {from=MaterialExpressionVectorParameter_10, output=Domain RGB / Domain R}]
   properties {
    convertInputs [{
      type=Scalar
      expressionInput {
       expression=&self:MaterialExpressionVectorParameter_10
       inputName=None
       mask=1
       maskA=0
       maskB=0
       maskG=1
       maskR=0
       outputIndex=2
      }
     }]
    convertOutputs [{type=Vector3}]
   }
  }
  MaterialExpressionConvert_8 {
   class=MaterialExpressionConvert
   inputs [
    None {from=MaterialExpressionFunctionInput_1}
    None {from=""}
   ]
   properties {
    convertInputs [
     {
      type=Vector2
      expressionInput {
       expression=&self:MaterialExpressionFunctionInput_1
       inputName=None
       mask=0
       maskA=0
       maskB=0
       maskG=0
       maskR=0
       outputIndex=0
      }
     }
     {type=Scalar, expressionInput {expression=null, inputName=None, mask=0, maskA=0, maskB=0, maskG=0, maskR=0, outputIndex=0}}
    ]
    convertOutputs [{type=Vector4, defaultValue=<rgba 0 0 0 1>}]
   }
  }
  MaterialExpressionConvert_9 {
   class=MaterialExpressionConvert
   inputs [
    None {from=MaterialExpressionFunctionInput_2}
    None {from=""}
   ]
   properties {
    convertInputs [
     {
      type=Vector3
      expressionInput {
       expression=&self:MaterialExpressionFunctionInput_2
       inputName=None
       mask=0
       maskA=0
       maskB=0
       maskG=0
       maskR=0
       outputIndex=0
      }
     }
     {type=Scalar, expressionInput {expression=null, inputName=None, mask=0, maskA=0, maskB=0, maskG=0, maskR=0, outputIndex=0}}
    ]
    convertOutputs [{type=Vector4, defaultValue=<rgba 0 0 0 1>}]
   }
  }
  MaterialExpressionFunctionInput_0 {
   class=MaterialExpressionFunctionInput
   inputs [Preview {from="", output=""}]
   properties {bUsePreviewValueAsDefault=true, inputName=Scalar In, inputType=FunctionInput_Scalar}
  }
  MaterialExpressionFunctionInput_1 {
   class=MaterialExpressionFunctionInput
   inputs [Preview {from="", output=""}]
   properties {bUsePreviewValueAsDefault=true, inputName=Vector2 In, inputType=FunctionInput_Vector2, sortPriority=10}
  }
  MaterialExpressionFunctionInput_2 {
   class=MaterialExpressionFunctionInput
   inputs [Preview {from="", output=""}]
   properties {bUsePreviewValueAsDefault=true, inputName=Vector3 In, sortPriority=20}
  }
  MaterialExpressionFunctionInput_3 {
   class=MaterialExpressionFunctionInput
   inputs [Preview {from="", output=""}]
   properties {bUsePreviewValueAsDefault=true, inputName=Vector4 In, inputType=FunctionInput_Vector4, sortPriority=30}
  }
  MaterialExpressionFunctionOutput_0 {
   class=MaterialExpressionFunctionOutput
   inputs [None {from=MaterialExpressionMaterialFunctionCall_4, output="Debug (V3)"}]
   properties {outputName="Scalar Debug (V3)"}
  }
  MaterialExpressionFunctionOutput_1 {
   class=MaterialExpressionFunctionOutput
   inputs [None {from=MaterialExpressionMaterialFunctionCall_1, output="Debug (V3)"}]
   properties {outputName="Vector3 Debug (V3)", sortPriority=20}
  }
  MaterialExpressionFunctionOutput_2 {
   class=MaterialExpressionFunctionOutput
   inputs [None {from=MaterialExpressionMaterialFunctionCall_0, output="Debug (V3)"}]
   properties {outputName="Vector2 Debug (V3)", sortPriority=10}
  }
  MaterialExpressionFunctionOutput_5 {
   class=MaterialExpressionFunctionOutput
   inputs [None {from=MaterialExpressionMaterialFunctionCall_2, output="Debug (V3)"}]
   properties {outputName="Vector4 Debug (V3)", sortPriority=30}
  }
  MaterialExpressionLinearInterpolate_0 {
   class=MaterialExpressionLinearInterpolate
   inputs [
    A {from=MaterialExpressionConvert_12}
    B {from=MaterialExpressionConvert_11}
    Alpha {from=MaterialExpressionRound_2}
   ]
  }
  MaterialExpressionLinearInterpolate_1 {
   class=MaterialExpressionLinearInterpolate
   inputs [
    A {from=MaterialExpressionConvert_1}
    B {from=MaterialExpressionConvert_0}
    Alpha {from=MaterialExpressionRound_0}
   ]
  }
  MaterialExpressionLinearInterpolate_2 {
   class=MaterialExpressionLinearInterpolate
   inputs [
    A {from=MaterialExpressionConvert_16}
    B {from=MaterialExpressionConvert_13}
    Alpha {from=MaterialExpressionRound_3}
   ]
  }
  MaterialExpressionLinearInterpolate_4 {
   class=MaterialExpressionLinearInterpolate
   inputs [
    A {from=MaterialExpressionConvert_18}
    B {from=MaterialExpressionConvert_17}
    Alpha {from=MaterialExpressionRound_4}
   ]
  }
  MaterialExpressionLinearInterpolate_6 {
   class=MaterialExpressionLinearInterpolate
   inputs [
    A {from=MaterialExpressionConvert_20}
    B {from=MaterialExpressionConvert_19}
    Alpha {from=MaterialExpressionRound_4}
   ]
  }
  MaterialExpressionLinearInterpolate_7 {
   class=MaterialExpressionLinearInterpolate
   inputs [
    A {from=MaterialExpressionConvert_22}
    B {from=MaterialExpressionConvert_21}
    Alpha {from=MaterialExpressionRound_6}
   ]
  }
  MaterialExpressionMaterialFunctionCall_0 {
   class=MaterialExpressionMaterialFunctionCall
   inputs [
    Signal In {from=""}
    //------ CHANNELS SWIZZLE ------// {from=""}
    Swizzle - Num Debug Channels {from=""}
    Swizzle - R Channel Out {from=""}
    Swizzle - G Channel Out {from=""}
    Swizzle - B Channel Out {from=""}
    //------ PER-CHANNEL REMAPS ------// {from=""}
    Enable Remaps {from=""}
    Remap - Invert Sign {from=""}
    "Remap - Change Domain (see tooltip)" {from=""}
    Remap - One Minus {from=""}
    Remap - Range OldMin {from=""}
    Remap - Range OldMax {from=""}
    Maths - Offset {from=""}
    Maths - Multiply {from=""}
    Maths - Power {from=""}
    Color - RGB Tint {from=""}
    //------ PER-CHANNEL OUTPUT ------// {from=""}
    Output - sRGB2Linear {from=""}
   ]
   properties {materialFunction=&/MF_DebugSignal}
  }
  MaterialExpressionMaterialFunctionCall_1 {
   class=MaterialExpressionMaterialFunctionCall
   inputs [
    Signal In {from=""}
    //------ CHANNELS SWIZZLE ------// {from=""}
    Swizzle - Num Debug Channels {from=""}
    Swizzle - R Channel Out {from=""}
    Swizzle - G Channel Out {from=""}
    Swizzle - B Channel Out {from=""}
    //------ PER-CHANNEL REMAPS ------// {from=""}
    Enable Remaps {from=""}
    Remap - Invert Sign {from=""}
    "Remap - Change Domain (see tooltip)" {from=""}
    Remap - One Minus {from=""}
    Remap - Range OldMin {from=""}
    Remap - Range OldMax {from=""}
    Maths - Offset {from=""}
    Maths - Multiply {from=""}
    Maths - Power {from=""}
    Color - RGB Tint {from=""}
    //------ PER-CHANNEL OUTPUT ------// {from=""}
    Output - sRGB2Linear {from=""}
   ]
   properties {materialFunction=&MF_DebugSignal}
  }
  MaterialExpressionMaterialFunctionCall_2 {
   class=MaterialExpressionMaterialFunctionCall
   inputs [
    Signal In {from=""}
    //------ CHANNELS SWIZZLE ------// {from=""}
    Swizzle - Num Debug Channels {from=""}
    Swizzle - R Channel Out {from=""}
    Swizzle - G Channel Out {from=""}
    Swizzle - B Channel Out {from=""}
    //------ PER-CHANNEL REMAPS ------// {from=""}
    Enable Remaps {from=""}
    Remap - Invert Sign {from=""}
    "Remap - Change Domain (see tooltip)" {from=""}
    Remap - One Minus {from=""}
    Remap - Range OldMin {from=""}
    Remap - Range OldMax {from=""}
    Maths - Offset {from=""}
    Maths - Multiply {from=""}
    Maths - Power {from=""}
    Color - RGB Tint {from=""}
    //------ PER-CHANNEL OUTPUT ------// {from=""}
    Output - sRGB2Linear {from=""}
   ]
   properties {materialFunction=&MF_DebugSignal}
  }
  MaterialExpressionMaterialFunctionCall_4 {
   class=MaterialExpressionMaterialFunctionCall
   inputs [
    Signal In {from=""}
    //------ CHANNELS SWIZZLE ------// {from=""}
    Swizzle - Num Debug Channels {from=""}
    Swizzle - R Channel Out {from=""}
    Swizzle - G Channel Out {from=""}
    Swizzle - B Channel Out {from=""}
    //------ PER-CHANNEL REMAPS ------// {from=""}
    Enable Remaps {from=""}
    Remap - Invert Sign {from=""}
    "Remap - Change Domain (see tooltip)" {from=""}
    Remap - One Minus {from=""}
    Remap - Range OldMin {from=""}
    Remap - Range OldMax {from=""}
    Maths - Offset {from=""}
    Maths - Multiply {from=""}
    Maths - Power {from=""}
    Color - RGB Tint {from=""}
    //------ PER-CHANNEL OUTPUT ------// {from=""}
    Output - sRGB2Linear {from=""}
   ]
   properties {materialFunction=&MF_DebugSignal}
  }
  MaterialExpressionMultiply_0 {
   class=MaterialExpressionMultiply
   inputs [
    A {from=MaterialExpressionVectorParameter_4, output=RGB}
    B {from=MaterialExpressionVectorParameter_4, output=RGB}
   ]
  }
  MaterialExpressionMultiply_3 {
   class=MaterialExpressionMultiply
   inputs [
    A {from=MaterialExpressionVectorParameter_5, output=RGB}
    B {from=MaterialExpressionVectorParameter_5, output=RGB}
   ]
  }
  MaterialExpressionNamedRerouteDeclaration_0 {
   class=MaterialExpressionNamedRerouteDeclaration
   inputs [None {from=MaterialExpressionLinearInterpolate_7}]
   properties {name="Remap - Change Domain (V3)", nodeColor=<rgba 1 0.2783 0 1>}
  }
  MaterialExpressionNamedRerouteDeclaration_1 {
   class=MaterialExpressionNamedRerouteDeclaration
   inputs [None {from=MaterialExpressionVectorParameter_0, output=Num Debug Channels}]
   properties {name="Swizzle - Num Debug Channels (S)", nodeColor=<rgba 1 0 0.4688 1>}
  }
  MaterialExpressionNamedRerouteDeclaration_12 {
   class=MaterialExpressionNamedRerouteDeclaration
   inputs [None {from=MaterialExpressionLinearInterpolate_4}]
   properties {name="Remap - Range OldMin (V3)", nodeColor=<rgba 1 0.2783 0 1>}
  }
  MaterialExpressionNamedRerouteDeclaration_15 {
   class=MaterialExpressionNamedRerouteDeclaration
   inputs [None {from=MaterialExpressionLinearInterpolate_6}]
   properties {name="Remap - Range OldMax (V3)", nodeColor=<rgba 1 0.2783 0 1>}
  }
  MaterialExpressionNamedRerouteDeclaration_17 {
   class=MaterialExpressionNamedRerouteDeclaration
   inputs [None {from=MaterialExpressionAdd_2}]
   properties {name="Maths - Offset (V3)", nodeColor=<rgba 0 1 0.1565 1>}
  }
  MaterialExpressionNamedRerouteDeclaration_18 {
   class=MaterialExpressionNamedRerouteDeclaration
   inputs [None {from=MaterialExpressionMultiply_3}]
   properties {name="Maths - Multiply (V3)", nodeColor=<rgba 0 1 0.1565 1>}
  }
  MaterialExpressionNamedRerouteDeclaration_19 {
   class=MaterialExpressionNamedRerouteDeclaration
   inputs [None {from=MaterialExpressionMultiply_0}]
   properties {name="Maths - Power (V3)", nodeColor=<rgba 0 1 0.1565 1>}
  }
  MaterialExpressionNamedRerouteDeclaration_2 {
   class=MaterialExpressionNamedRerouteDeclaration
   inputs [None {from=MaterialExpressionVectorParameter_0, output=R Channel Out}]
   properties {name="Swizzle - R Channel Out (S)", nodeColor=<rgba 1 0 0.4687 1>}
  }
  MaterialExpressionNamedRerouteDeclaration_20 {
   class=MaterialExpressionNamedRerouteDeclaration
   inputs [None {from=MaterialExpressionVectorParameter_6, output=RGB}]
   properties {name="Color - RGB Tint (V3)", nodeColor=<rgba 1 0 0 1>}
  }
  MaterialExpressionNamedRerouteDeclaration_21 {
   class=MaterialExpressionNamedRerouteDeclaration
   inputs [None {from=MaterialExpressionStaticBoolParameter_0}]
   properties {name="Enable Remaps (SB)", nodeColor=<rgba 0.713 1 0 1>}
  }
  MaterialExpressionNamedRerouteDeclaration_22 {
   class=MaterialExpressionNamedRerouteDeclaration
   inputs [None {from=MaterialExpressionConvert_2}]
   properties {name="Scalar In (V4)", nodeColor=<rgba 1 1 1 1>}
  }
  MaterialExpressionNamedRerouteDeclaration_23 {
   class=MaterialExpressionNamedRerouteDeclaration
   inputs [None {from=MaterialExpressionConvert_8}]
   properties {name="Vector2 In (V4)", nodeColor=<rgba 1 1 1 1>}
  }
  MaterialExpressionNamedRerouteDeclaration_24 {
   class=MaterialExpressionNamedRerouteDeclaration
   inputs [None {from=MaterialExpressionConvert_9}]
   properties {name="Vector3 In (V4)", nodeColor=<rgba 1 1 1 1>}
  }
  MaterialExpressionNamedRerouteDeclaration_25 {
   class=MaterialExpressionNamedRerouteDeclaration
   inputs [None {from=MaterialExpressionFunctionInput_3}]
   properties {name="Vector4 In (V4)", nodeColor=<rgba 1 1 1 1>}
  }
  MaterialExpressionNamedRerouteDeclaration_27 {
   class=MaterialExpressionNamedRerouteDeclaration
   inputs [None {from=MaterialExpressionLinearInterpolate_0}]
   properties {name="Output - sRGB2Linear (V3)", nodeColor=<rgba 0 0.3826 1 1>}
  }
  MaterialExpressionNamedRerouteDeclaration_3 {
   class=MaterialExpressionNamedRerouteDeclaration
   inputs [None {from=MaterialExpressionVectorParameter_0, output=G Channel Out}]
   properties {name="Swizzle - G Channel Out (S)", nodeColor=<rgba 1 0 0.4688 1>}
  }
  MaterialExpressionNamedRerouteDeclaration_4 {
   class=MaterialExpressionNamedRerouteDeclaration
   inputs [None {from=MaterialExpressionVectorParameter_0, output=B Channel Out}]
   properties {name="Swizzle - B Channel Out (S)", nodeColor=<rgba 1 0 0.4688 1>}
  }
  MaterialExpressionNamedRerouteDeclaration_6 {
   class=MaterialExpressionNamedRerouteDeclaration
   inputs [None {from=MaterialExpressionLinearInterpolate_2}]
   properties {name="Remap - One Minus (V3)", nodeColor=<rgba 1 0.2783 0 1>}
  }
  MaterialExpressionNamedRerouteDeclaration_7 {
   class=MaterialExpressionNamedRerouteDeclaration
   inputs [None {from=MaterialExpressionLinearInterpolate_1}]
   properties {name="Remap - Invert Sign (V3)", nodeColor=<rgba 1 0.2783 0 1>}
  }
  MaterialExpressionNamedRerouteUsage_0 {class=MaterialExpressionNamedRerouteUsage}
  MaterialExpressionNamedRerouteUsage_1 {class=MaterialExpressionNamedRerouteUsage}
  MaterialExpressionNamedRerouteUsage_100 {class=MaterialExpressionNamedRerouteUsage}
  MaterialExpressionNamedRerouteUsage_101 {class=MaterialExpressionNamedRerouteUsage}
  MaterialExpressionNamedRerouteUsage_102 {class=MaterialExpressionNamedRerouteUsage}
  MaterialExpressionNamedRerouteUsage_103 {class=MaterialExpressionNamedRerouteUsage}
  MaterialExpressionNamedRerouteUsage_104 {class=MaterialExpressionNamedRerouteUsage}
  MaterialExpressionNamedRerouteUsage_105 {class=MaterialExpressionNamedRerouteUsage}
  MaterialExpressionNamedRerouteUsage_106 {class=MaterialExpressionNamedRerouteUsage}
  MaterialExpressionNamedRerouteUsage_107 {class=MaterialExpressionNamedRerouteUsage}
  MaterialExpressionNamedRerouteUsage_108 {class=MaterialExpressionNamedRerouteUsage}
  MaterialExpressionNamedRerouteUsage_109 {class=MaterialExpressionNamedRerouteUsage}
  MaterialExpressionNamedRerouteUsage_11 {class=MaterialExpressionNamedRerouteUsage}
  MaterialExpressionNamedRerouteUsage_110 {class=MaterialExpressionNamedRerouteUsage}
  MaterialExpressionNamedRerouteUsage_111 {class=MaterialExpressionNamedRerouteUsage}
  MaterialExpressionNamedRerouteUsage_112 {class=MaterialExpressionNamedRerouteUsage}
  MaterialExpressionNamedRerouteUsage_113 {class=MaterialExpressionNamedRerouteUsage}
  MaterialExpressionNamedRerouteUsage_114 {class=MaterialExpressionNamedRerouteUsage}
  MaterialExpressionNamedRerouteUsage_115 {class=MaterialExpressionNamedRerouteUsage}
  MaterialExpressionNamedRerouteUsage_116 {class=MaterialExpressionNamedRerouteUsage}
  MaterialExpressionNamedRerouteUsage_117 {class=MaterialExpressionNamedRerouteUsage}
  MaterialExpressionNamedRerouteUsage_118 {class=MaterialExpressionNamedRerouteUsage}
  MaterialExpressionNamedRerouteUsage_119 {class=MaterialExpressionNamedRerouteUsage}
  MaterialExpressionNamedRerouteUsage_120 {class=MaterialExpressionNamedRerouteUsage}
  MaterialExpressionNamedRerouteUsage_121 {class=MaterialExpressionNamedRerouteUsage}
  MaterialExpressionNamedRerouteUsage_122 {class=MaterialExpressionNamedRerouteUsage}
  MaterialExpressionNamedRerouteUsage_123 {class=MaterialExpressionNamedRerouteUsage}
  MaterialExpressionNamedRerouteUsage_124 {class=MaterialExpressionNamedRerouteUsage}
  MaterialExpressionNamedRerouteUsage_125 {class=MaterialExpressionNamedRerouteUsage}
  MaterialExpressionNamedRerouteUsage_126 {class=MaterialExpressionNamedRerouteUsage}
  MaterialExpressionNamedRerouteUsage_127 {class=MaterialExpressionNamedRerouteUsage}
  MaterialExpressionNamedRerouteUsage_128 {class=MaterialExpressionNamedRerouteUsage}
  MaterialExpressionNamedRerouteUsage_129 {class=MaterialExpressionNamedRerouteUsage}
  MaterialExpressionNamedRerouteUsage_13 {class=MaterialExpressionNamedRerouteUsage}
  MaterialExpressionNamedRerouteUsage_130 {class=MaterialExpressionNamedRerouteUsage}
  MaterialExpressionNamedRerouteUsage_131 {class=MaterialExpressionNamedRerouteUsage}
  MaterialExpressionNamedRerouteUsage_15 {class=MaterialExpressionNamedRerouteUsage}
  MaterialExpressionNamedRerouteUsage_16 {class=MaterialExpressionNamedRerouteUsage}
  MaterialExpressionNamedRerouteUsage_17 {class=MaterialExpressionNamedRerouteUsage}
  MaterialExpressionNamedRerouteUsage_18 {class=MaterialExpressionNamedRerouteUsage}
  MaterialExpressionNamedRerouteUsage_19 {class=MaterialExpressionNamedRerouteUsage}
  MaterialExpressionNamedRerouteUsage_2 {class=MaterialExpressionNamedRerouteUsage}
  MaterialExpressionNamedRerouteUsage_20 {class=MaterialExpressionNamedRerouteUsage}
  MaterialExpressionNamedRerouteUsage_3 {class=MaterialExpressionNamedRerouteUsage}
  MaterialExpressionNamedRerouteUsage_38 {class=MaterialExpressionNamedRerouteUsage}
  MaterialExpressionNamedRerouteUsage_4 {class=MaterialExpressionNamedRerouteUsage}
  MaterialExpressionNamedRerouteUsage_5 {class=MaterialExpressionNamedRerouteUsage}
  MaterialExpressionNamedRerouteUsage_56 {class=MaterialExpressionNamedRerouteUsage}
  MaterialExpressionNamedRerouteUsage_6 {class=MaterialExpressionNamedRerouteUsage}
  MaterialExpressionNamedRerouteUsage_74 {class=MaterialExpressionNamedRerouteUsage}
  MaterialExpressionNamedRerouteUsage_75 {class=MaterialExpressionNamedRerouteUsage}
  MaterialExpressionNamedRerouteUsage_76 {class=MaterialExpressionNamedRerouteUsage}
  MaterialExpressionNamedRerouteUsage_77 {class=MaterialExpressionNamedRerouteUsage}
  MaterialExpressionNamedRerouteUsage_79 {class=MaterialExpressionNamedRerouteUsage}
  MaterialExpressionNamedRerouteUsage_8 {class=MaterialExpressionNamedRerouteUsage}
  MaterialExpressionNamedRerouteUsage_90 {class=MaterialExpressionNamedRerouteUsage}
  MaterialExpressionNamedRerouteUsage_91 {class=MaterialExpressionNamedRerouteUsage}
  MaterialExpressionNamedRerouteUsage_92 {class=MaterialExpressionNamedRerouteUsage}
  MaterialExpressionNamedRerouteUsage_93 {class=MaterialExpressionNamedRerouteUsage}
  MaterialExpressionNamedRerouteUsage_94 {class=MaterialExpressionNamedRerouteUsage}
  MaterialExpressionNamedRerouteUsage_95 {class=MaterialExpressionNamedRerouteUsage}
  MaterialExpressionNamedRerouteUsage_96 {class=MaterialExpressionNamedRerouteUsage}
  MaterialExpressionNamedRerouteUsage_97 {class=MaterialExpressionNamedRerouteUsage}
  MaterialExpressionNamedRerouteUsage_99 {class=MaterialExpressionNamedRerouteUsage}
  MaterialExpressionRound_0 {class=MaterialExpressionRound, inputs [None {from=MaterialExpressionSaturate_0, output=""}], properties {}}
  MaterialExpressionRound_2 {class=MaterialExpressionRound, inputs [None {from=MaterialExpressionSaturate_2, output=""}], properties {}}
  MaterialExpressionRound_3 {class=MaterialExpressionRound, inputs [None {from=MaterialExpressionSaturate_3, output=""}], properties {}}
  MaterialExpressionRound_4 {class=MaterialExpressionRound, inputs [None {from=MaterialExpressionSaturate_4, output=""}], properties {}}
  MaterialExpressionRound_6 {class=MaterialExpressionRound, inputs [None {from=MaterialExpressionSaturate_6, output=""}], properties {}}
  MaterialExpressionSaturate_0 {class=MaterialExpressionSaturate, inputs [None {from=MaterialExpressionVectorParameter_9, output=Enable Per Channel}]}
  MaterialExpressionSaturate_2 {class=MaterialExpressionSaturate, inputs [None {from=MaterialExpressionVectorParameter_8, output=Enable Per Channel}]}
  MaterialExpressionSaturate_3 {class=MaterialExpressionSaturate, inputs [None {from=MaterialExpressionVectorParameter_11, output=Enable Per Channel}]}
  MaterialExpressionSaturate_4 {class=MaterialExpressionSaturate, inputs [None {from=MaterialExpressionVectorParameter_2, output=Enable Per Channel}]}
  MaterialExpressionSaturate_6 {class=MaterialExpressionSaturate, inputs [None {from=MaterialExpressionVectorParameter_10, output=Enable Per Channel}]}
  MaterialExpressionStaticBoolParameter_0 {class=MaterialExpressionStaticBoolParameter, inputs [], properties {group=zz_SIGNAL DEBUG, parameterName=Enable Remaps, sortPriority=100}}
  MaterialExpressionVectorParameter_0 {
   class=MaterialExpressionVectorParameter
   properties {
    desc="- Num Debug Channels -> 1, 2 or 3 (R, RG, RGB)\r\n\r\n- x Channel Out -> 0: R, 1: G, 2: B\r\n"
    group=zz_SIGNAL DEBUG
    parameterName=Channels Swizzle
    sortPriority=20
    channelNames {a=B Channel Out, b=G Channel Out, g=R Channel Out, r=Num Debug Channels}
    defaultValue=<rgba 3 0 1 2>
   }
  }
  MaterialExpressionVectorParameter_10 {
   class=MaterialExpressionVectorParameter
   properties {
    desc="-1 :  [0, 1] to [-1, 1]\r\n0  : no change\r\n1 : [-1, 1] to [0, 1]\r\n"
    group=zz_SIGNAL DEBUG
    parameterName="Remap - Change Domain (see tooltip)"
    sortPriority=120
    channelNames {a=Domain B, b=Domain G, g=Domain RGB / Domain R, r=Enable Per Channel}
   }
  }
  MaterialExpressionVectorParameter_11 {
   class=MaterialExpressionVectorParameter
   properties {
    group=zz_SIGNAL DEBUG
    parameterName=Remap - One Minus
    sortPriority=130
    channelNames {a=One Minus B, b=One Minus G, g=One Minus RGB / One Minus R, r=Enable Per Channel}
   }
  }
  MaterialExpressionVectorParameter_2 {
   class=MaterialExpressionVectorParameter
   properties {
    group=zz_SIGNAL DEBUG
    parameterName=Remap - Range OldMin
    sortPriority=140
    channelNames {a=OldMin B, b=OldMin G, g=OldMin RGB / OldMin R, r=Enable Per Channel}
   }
  }
  MaterialExpressionVectorParameter_3 {
   class=MaterialExpressionVectorParameter
   properties {
    group=zz_SIGNAL DEBUG
    parameterName=Remap - Range OldMax
    sortPriority=150
    channelNames {a=OldMax B, b=OldMax G, g=OldMax RGB / OldMax R, r=n/a}
    defaultValue=<rgba 0 1 1 1>
   }
  }
  MaterialExpressionVectorParameter_4 {
   class=MaterialExpressionVectorParameter
   properties {
    group=zz_SIGNAL DEBUG
    parameterName=Maths - Power
    sortPriority=230
    channelNames {a=Power RGB, b=Power B, g=Power G, r=Power R}
    defaultValue=<rgba 1 1 1 1>
   }
  }
  MaterialExpressionVectorParameter_5 {
   class=MaterialExpressionVectorParameter
   properties {
    group=zz_SIGNAL DEBUG
    parameterName=Maths - Multiply
    sortPriority=220
    channelNames {a=Multiply RGB, b=Multiply B, g=Multiply G, r=Multiply R}
    defaultValue=<rgba 1 1 1 1>
   }
  }
  MaterialExpressionVectorParameter_6 {
   class=MaterialExpressionVectorParameter
   properties {
    group=zz_SIGNAL DEBUG
    parameterName=Color - RGB Tint
    sortPriority=240
    channelNames {a=n/a, b=Tint B, g=Tint G, r=Tint R}
    defaultValue=<rgba 1 1 1 0>
   }
  }
  MaterialExpressionVectorParameter_7 {
   class=MaterialExpressionVectorParameter
   properties {group=zz_SIGNAL DEBUG, parameterName=Maths - Offset, sortPriority=210, channelNames {a=Offset RGB, b=Offset B, g=Offset G, r=Offset R}}
  }
  MaterialExpressionVectorParameter_8 {
   class=MaterialExpressionVectorParameter
   properties {
    group=zz_SIGNAL DEBUG
    parameterName="Output - sRGB->Linear"
    sortPriority=10
    channelNames {a=Convert B, b=Convert G, g=Convert RGB / Convert R, r=Enable Per Channel}
   }
  }
  MaterialExpressionVectorParameter_9 {
   class=MaterialExpressionVectorParameter
   properties {
    group=zz_SIGNAL DEBUG
    parameterName=Remap - Invert Sign
    sortPriority=110
    channelNames {a=Invert Sign B, b=Invert Sign G, g=Invert Sign RGB / Invert Sign R, r=Enable Per Channel}
   }
  }
 }}
