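# Material function for the F_MED_FermiumStraw character body that optionally replaces a material with a debug view.
# The "Enable Debug" static switch chooses between passing Material In through and a fresh attribute set.
# That set shows MFW_DebugSignal's Vector3 output as emissive, with black base color, zeroed surface response, full opacity and a flat normal.
# The declared Debug Signal input is left unconnected, as are the MFW_DebugSignal inputs.
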MaterialFunction /BRCosmetics/Characters/Player/Female/Medium/Bodies/F_MED_FermiumStraw/Materials/MaterialFunctions/DEV/MF_AddDebugMaterial
graph {nodes [
  FunctionInput() {bUsePreviewValueAsDefault=true, inputName=Debug Signal, sortPriority=10, previewValue=<1 0 1 1>}
  n1 = MaterialFunctionCall!(Scalar In, Vector2 In, Vector3 In, Vector4 In) {materialFunction=/BRCosmetics/Characters/Player/Female/Medium/Bodies/F_MED_FermiumStraw/Materials/MaterialFunctions/DEV/Dependencies/MFW_DebugSignal}
  n2 = Constant()
  n3 = Constant() {r=1}
  n4 = SetMaterialAttributes!(
   MaterialAttributes
   EmissiveColor=@n1."Vector3 Debug (V3)"
   BaseColor=@n2
   Metallic=@n2
   Specular=@n2
   Roughness=@n2
   Opacity=@n3
   OpacityMask=@n3
   Normal=Convert!() {convertOutputs [{type=Vector3, defaultValue=<rgba 0 0 1 1>}]}
   ShadingModel=ShadingModel()
  ) {attributeSetTypes [
    #1
    #2
    #3
    #4
    #5
    #6
    #7
    #8
    #9
   ]}
  FunctionOutput(None=StaticSwitchParameter(True=@n4, False=FunctionInput() {inputName=Material In, inputType=FunctionInput_MaterialAttributes}) {group=zz_SIGNAL DEBUG, parameterName=Enable Debug, sortPriority=-1000}) {outputName="Material Out (MA)"}
 ]}
