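# Material function in the FermiumStraw character's DEV folder that visualizes UV coordinates as a debug color, flagging values outside the 0 to 1 range.
# It takes a UV input, defaulting to the texture coordinate, into a single Custom HLSL node that feeds the vector output.
# In-range UVs display as red and green, while each of the eight out-of-range regions gets its own fixed blue-tinted color showing the overflow direction.
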MaterialFunction /BRCosmetics/Characters/Player/Female/Medium/Bodies/F_MED_FermiumStraw/Materials/MaterialFunctions/DEV/MF_UVDebug
graph {nodes [
  MaterialExpressionFunctionInput_0 = FunctionInput(Preview=TextureCoordinate()) {bUsePreviewValueAsDefault=true, inputName=UVs, inputType=FunctionInput_Vector2}
  n1 = Custom!(UVs=@MaterialExpressionFunctionInput_0) {
   code="int Uneg = 0;\r\nint Ulrg = 0;\r\nint Uvalid = 1;\r\nif (UVs.x < 0.0) {\r\n    Uneg = 1;\r\n    Uvalid = 0;\r\n}\r\nelse if (UVs.x > 1.0) {\r\n    Ulrg = 1;\r\n    Uvalid = 0;\r\n}\r\n\r\nint Vneg = 0;\r\nint Vlrg = 0;\r\nint Vvalid = 1;\r\nif (UVs.y < 0.0) {\r\n    Vneg = 1;\r\n    Vvalid = 0;\r\n}\r\nelse if (UVs.y > 1.0) {\r\n    Vlrg = 1;\r\n    Vvalid = 0;\r\n}\r\n\r\nif (Uvalid == 1 &&&& Vvalid == 1) {\r\n    return float3(UVs, 0.0);\r\n}\r\nelse if (Uvalid == 1 &&&& Vvalid == 0) {\r\n    if (Vneg == 1) {\r\n        return float3(0.0, 0.05, 0.5);\r\n    }\r\n    else {\r\n        return float3(0.0, 0.4, 1.0);\r\n    }\r\n}\r\nelse if (Uvalid == 0 &&&& Vvalid == 1) {\r\n    if (Uneg == 1) {\r\n        return float3(0.05, 0.0, 0.5);\r\n    }\r\n    else {\r\n        return float3(0.4, 0.0, 1.0);\r\n    }\r\n}\r\nelse if (Uneg == 1) {\r\n    if (Vneg == 1) {\r\n        return float3(0.0, 0.0, 0.5);\r\n    }\r\n    else {\r\n        return float3(0.1, 0.7, 1.0);\r\n    }\r\n}\r\nelse {\r\n    if (Vneg == 1) {\r\n        return float3(0.7, 0.1, 1.0);\r\n    }\r\n    else {\r\n        return float3(0.7, 0.7, 1.0);\r\n    }\r\n}"
   showCode=true
   inputs [{
     inputName=UVs
     input {expression=&self:MaterialExpressionFunctionInput_0}
    }]
  }
  FunctionOutput(None=@n1) {outputName="DebugColor (V3)"}
 ]}
