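# Debug material function, in a FermiumStraw character DEV folder, that unwraps a mesh into its UV layout and samples a texture on it.
# A rounded, clamped index input switches between UV channels 0, 1 and 2, feeding the UV output through named reroutes.
# World position offset collapses vertices to the object position, then spreads them on the XY plane by centered, scaled UVs.
# A replaceable texture input provides the RGB output.
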MaterialFunction /BRCosmetics/Characters/Player/Female/Medium/Bodies/F_MED_FermiumStraw/Materials/MaterialFunctions/DEV/MF_UVTileDebug
graph {nodes [
  n1 = FunctionInput() {bUsePreviewValueAsDefault=true, inputName=UV Index, inputType=FunctionInput_Scalar}
  MaterialExpressionTextureCoordinate_1 = TextureCoordinate()
  MaterialExpressionTextureCoordinate_0 = TextureCoordinate() {coordinateIndex=1}
  MaterialExpressionTextureCoordinate_2 = TextureCoordinate() {coordinateIndex=2}
  n2 = Switch!(
   SwitchValue=Round(None=Clamp(None=@n1) {maxDefault=2})
   Default
   UV0=@MaterialExpressionTextureCoordinate_1
   UV1=@MaterialExpressionTextureCoordinate_0
   UV2=@MaterialExpressionTextureCoordinate_2
  ) {inputs [
    {
     inputName=UV0
     input {
      expression=&self:MaterialExpressionTextureCoordinate_1
      inputName=None
      mask=0
      maskA=0
      maskB=0
      maskG=0
      maskR=0
      outputIndex=0
     }
    }
    {
     inputName=UV1
     input {
      expression=&self:MaterialExpressionTextureCoordinate_0
      inputName=None
      mask=0
      maskA=0
      maskB=0
      maskG=0
      maskR=0
      outputIndex=0
     }
    }
    {
     inputName=UV2
     input {
      expression=&self:MaterialExpressionTextureCoordinate_2
      inputName=None
      mask=0
      maskA=0
      maskB=0
      maskG=0
      maskR=0
      outputIndex=0
     }
    }
   ]}
  FunctionOutput(None=NamedRerouteDeclaration(None=@n2) {name="UV (V2)", nodeColor=<rgba 0 1 0.9766 1>}) {outputName="UV (V2)"}
  n3 = FunctionInput() {bUsePreviewValueAsDefault=true, inputName=WPO Scale, inputType=FunctionInput_Scalar, sortPriority=30, previewValue=<100 0 0 1>}
  n4 = Multiply(A=Subtract(A=NamedRerouteUsage()) {constB=0.5}, B=@n3) {constB=100}
  n5 = Constant()
  n6 = ObjectPositionWS()
  n7 = WorldPosition()
  FunctionOutput(None=Add(A=AppendVector(A=@n4, B=@n5), B=Subtract(A=@n6, B=@n7.XYZ))) {outputName="WPO (V3)", sortPriority=20}
  n8 = TextureObject() {texture=&T_BlacksRGB}
  FunctionOutput(None=TextureSample(
    UVs=NamedRerouteUsage()
    Tex=FunctionInput(Preview=@n8) {bUsePreviewValueAsDefault=true, inputName=Texture, inputType=FunctionInput_Texture2D, sortPriority=20}
   ).RGB) {outputName="Texture (V3)", sortPriority=10}
 ]}
